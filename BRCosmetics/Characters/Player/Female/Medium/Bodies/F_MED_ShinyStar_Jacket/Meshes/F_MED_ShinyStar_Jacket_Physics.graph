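# Physics asset for the F_MED_ShinyStar_Jacket female medium body, giving five dyn_jacket bone chains (A center, left and right; B left and right) bodies for secondary motion.
# Each chain pairs a tiny MSTR anchor capsule, angularly free against spine_04, with a longer 01 capsule constrained to that anchor by tight swing limits and a zero twist limit.
# Three large spine capsules form the torso volume, linked spine_02 through spine_04.
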PhysicsAsset /BRCosmetics/Characters/Player/Female/Medium/Bodies/F_MED_ShinyStar_Jacket/Meshes/F_MED_ShinyStar_Jacket_Physics
bodies {
 dyn_jacketA_c_01 [{shapeType=Capsule, length=6.2102, radius=1.8166, shapeName=dyn_jacketA_c_01_capsule, center=<4.4341 0 0>, rotation=<rot 90 0 0>}]
 dyn_jacketA_c_MSTR [{shapeType=Capsule, length=0.5, radius=0.5, shapeName=dyn_jacketA_c_MSTR_capsule}]
 dyn_jacketA_l_01 [{shapeType=Capsule, length=6.2102, radius=1.8166, shapeName=dyn_jacketA_l_01_capsule, center=<4.4341 0 0>, rotation=<rot 90 0 0>}]
 dyn_jacketA_l_MSTR [{shapeType=Capsule, length=0.5, radius=0.5, shapeName=dyn_jacketA_l_MSTR_capsule}]
 dyn_jacketA_r_01 [{shapeType=Capsule, length=6.2102, radius=1.8166, shapeName=dyn_jacketA_r_01_capsule, center=<4.4341 0 0>, rotation=<rot 90 0 0>}]
 dyn_jacketA_r_MSTR [{shapeType=Capsule, length=0.5, radius=0.5, shapeName=dyn_jacketA_r_MSTR_capsule}]
 dyn_jacketB_l_01 [{shapeType=Capsule, length=6.2102, radius=1.8166, shapeName=dyn_jacketB_l_01_capsule, center=<4.4341 0 0>, rotation=<rot 90 0 0>}]
 dyn_jacketB_l_MSTR [{shapeType=Capsule, length=0.5, radius=0.5, shapeName=dyn_jacketB_l_MSTR_capsule}]
 dyn_jacketB_r_01 [{shapeType=Capsule, length=6.2102, radius=1.8166, shapeName=dyn_jacketB_r_01_capsule, center=<4.4341 0 0>, rotation=<rot 90 0 0>}]
 dyn_jacketB_r_MSTR [{shapeType=Capsule, length=0.5, radius=0.5, shapeName=dyn_jacketB_r_MSTR_capsule}]
 spine_02 [{shapeType=Capsule, length=3, radius=7.8839, shapeName=spine_02_capsule}]
 spine_03 [{shapeType=Capsule, length=1.3547, radius=9.3493, shapeName=spine_03_capsule, center=<0 2.865 0>}]
 spine_04 [{shapeType=Capsule, length=7.9027, radius=6.9012, center=<0 2.3663 0>}]
}
constraints [
 {bone1Name=spine_04, bone2Name=spine_03, swing1LimitDegrees=45, swing2LimitDegrees=45, twistLimitDegrees=45}
 {
  bone1Name=spine_03
  bone2Name=spine_02
  swing1LimitDegrees=45
  swing1Motion=Free
  swing2LimitDegrees=45
  swing2Motion=Free
  twistLimitDegrees=45
  twistMotion=Free
 }
 {bone1Name=dyn_jacketB_r_01, bone2Name=dyn_jacketB_r_MSTR, swing1LimitDegrees=10, swing2LimitDegrees=20, twistLimitDegrees=0}
 {
  bone1Name=dyn_jacketB_r_MSTR
  bone2Name=spine_04
  swing1LimitDegrees=45
  swing1Motion=Free
  swing2LimitDegrees=45
  swing2Motion=Free
  twistLimitDegrees=45
  twistMotion=Free
 }
 {bone1Name=dyn_jacketB_l_01, bone2Name=dyn_jacketB_l_MSTR, swing1LimitDegrees=10, swing2LimitDegrees=20, twistLimitDegrees=0}
 {
  bone1Name=dyn_jacketB_l_MSTR
  bone2Name=spine_04
  swing1LimitDegrees=45
  swing1Motion=Free
  swing2LimitDegrees=45
  swing2Motion=Free
  twistLimitDegrees=45
  twistMotion=Free
 }
 {bone1Name=dyn_jacketA_r_01, bone2Name=dyn_jacketA_r_MSTR, swing1LimitDegrees=8, swing2LimitDegrees=30, twistLimitDegrees=0}
 {
  bone1Name=dyn_jacketA_r_MSTR
  bone2Name=spine_04
  swing1LimitDegrees=45
  swing1Motion=Free
  swing2LimitDegrees=45
  swing2Motion=Free
  twistLimitDegrees=45
  twistMotion=Free
 }
 {bone1Name=dyn_jacketA_l_01, bone2Name=dyn_jacketA_l_MSTR, swing1LimitDegrees=8, swing2LimitDegrees=30, twistLimitDegrees=0}
 {
  bone1Name=dyn_jacketA_l_MSTR
  bone2Name=spine_04
  swing1LimitDegrees=45
  swing1Motion=Free
  swing2LimitDegrees=45
  swing2Motion=Free
  twistLimitDegrees=45
  twistMotion=Free
 }
 {bone1Name=dyn_jacketA_c_01, bone2Name=dyn_jacketA_c_MSTR, swing1LimitDegrees=10, swing2LimitDegrees=20, twistLimitDegrees=0}
 {
  bone1Name=dyn_jacketA_c_MSTR
  bone2Name=spine_04
  swing1LimitDegrees=45
  swing1Motion=Free
  swing2LimitDegrees=45
  swing2Motion=Free
  twistLimitDegrees=45
  twistMotion=Free
 }
]
